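# Translucent unlit material for arc lightning on the Speedy Midnight backpack, built for Niagara mesh particles.
# A panning plasma texture drives HDR-tinted emissive and opacity, gated by two stepped, exponent-sharpened gradient sweeps and faded at strand ends, with depth fade off mobile.
# World position offset combines SplineThicken width with sine wobble strongest mid-strand.
# Vertex red varies timing and tiling; green blends in a secondary alpha.
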Material /BRCosmetics/Accessories/FORT_Backpacks/M_MED_Speedy_Midnight/FX/M_SpeedyArcLightning
graph {
 properties {
  bEnableResponsiveAA=true
  bUsedWithNiagaraMeshParticles=true
  blendMode=BLEND_Translucent
  pixelDepthOffsetMode=PDOM_Legacy
  shadingModel=MSM_Unlit
  twoSided=true
  usedShadingModels=Unlit
 }
 nodes [
  n1 = TextureCoordinate() {vTiling=0.5}
  n2 = VertexColor()
  n3 = AppendVector(A=Constant() {r=1}, B=LinearInterpolate(Alpha=@n2.R) {constA=0.5, constB=1.2})
  n4 = Add(A=@n2.R, B=Time())
  n5 = Add(A=ScalarParameter() {defaultValue=0.1, parameterName=PanSpeed}, B=@n2.R)
  n6 = TextureSample(UVs=Panner(Coordinate=Multiply(A=@n1, B=@n3), Time=Multiply(A=@n4, B=@n5)) {speedY=1}) {texture=/Game/Effects/Fort_Effects/Textures/Tile/T_Lightning_Plasma_Soft}
  n7 = Multiply(
   A=VectorParameter() {parameterName=HDRColor, defaultValue=<rgba 6 1.8699 0.4196 20>}.RGB
   B=LinearInterpolate(
    A=VectorParameter() {parameterName=EdgeColor, defaultValue=<rgba 0.5 0.1558 0.035 0>}.RGB
    B=VectorParameter() {parameterName=Color, defaultValue=<rgba 1 0.3116 0.0699 0>}.RGB
    Alpha=@n6.G
   )
  )
  MaterialFunctionCall(Emissive=MaterialFunctionCall(LDR Input=Clamp(None=@n7), HDR Tint=@n7, HDR Intensity Multiplier=ScalarParameter() {defaultValue=20, parameterName=HDRIntensity}) {materialFunction=&DeriveHDRfromLDR}.RemappedRange) {materialFunction=&MF_FX_EmissiveAdjustments}
  n8 = TextureCoordinate()
  n9 = MaterialFunctionCall(XY=TextureCoordinate() {coordinateIndex=1, uTiling=2}) {materialFunction=&Swizzle}
  n10 = VertexColor()
  n11 = Ceil(None=Multiply(A=Time() {period=1}) {constB=9})
  n12 = TextureSample(UVs=Panner(Coordinate=Add(A=@n9.YX, B=@n10.R), Time=Divide(A=@n11) {constB=3}) {speedY=-1}) {texture=/Game/Global/Textures/Default/Masks_and_Gradients/gradient_stasis}
  n13 = ScalarParameter() {defaultValue=8, parameterName=ArcExpControl}
  n14 = TextureCoordinate() {coordinateIndex=1, uTiling=2}
  n15 = Ceil(None=Multiply(A=Time() {period=1}) {constB=16})
  n16 = TextureSample(UVs=Panner(
    Coordinate=MaterialFunctionCall(XY=@n14) {materialFunction=&Swizzle}.YX
    Time=Divide(A=@n15) {constB=4}
   ) {speedY=1}) {texture=&gradient_stasis}
  n17 = ScalarParameter() {defaultValue=8, parameterName=ArcExpControl}
  n18 = Add(A=Multiply(A=Power(Base=@n12.G, Exp=@n13)) {constB=5}, B=Multiply(A=Power(Base=@n16.G, Exp=@n17)) {constB=5})
  n19 = Add(A=@n6.R, B=Multiply(A=@n6.B) {constB=0.2})
  n20 = ScalarParameter() {defaultValue=1, parameterName=Opacity}
  n21 = Multiply(A=Sine(None=ComponentMask(None=@n8) {g=true}) {period=2}, B=Multiply(A=Saturate(None=@n18), B=Multiply(A=@n19, B=@n20)))
  n22 = ScalarParameter() {defaultValue=10, parameterName=FadeDistance}
  Multiply(
   A=LinearInterpolate(B=ScalarParameter() {parameterName=SecondaryAlpha}, Alpha=VertexColor().G) {constA=1}
   B=ShadingPathSwitch(Default=DepthFade(Opacity=@n21, FadeDistance=@n22), Mobile=@n21) {desc=Migrated from FeatureLevelSwitch}
  )
  n23 = Multiply(A=ScalarParameter() {defaultValue=2, parameterName=Width}, B=ScalarParameter() {defaultValue=1, parameterName=WidthScale})
  n24 = TextureCoordinate()
  n25 = MaterialFunctionCall(
   WidthBase=@n23
   WidthTip=Multiply(A=@n23, B=ScalarParameter() {defaultValue=1, parameterName=WidthTipMult}) {constB=5}
   UVs for Thickness=Power(Base=@n24) {constExponent=0.6}
  ) {materialFunction=&SplineThicken}
  n26 = Constant() {r=0.5}
  n27 = SphereMask(A=ComponentMask(None=@n24) {g=true}, B=@n26, Radius=@n26) {hardnessPercent=50}
  n28 = TextureCoordinate()
  n29 = Multiply(A=Add(A=VertexColor().R, B=Time()), B=ScalarParameter() {defaultValue=5, parameterName=WobbleSpeed})
  n30 = ComponentMask(None=Add(A=@n28, B=Multiply(A=@n29) {constB=-0.6})) {g=true}
  n31 = Multiply(A=Sine(None=@n30) {period=0.2}) {constB=0.2}
  n32 = ComponentMask(None=Add(A=@n28, B=Multiply(A=@n29) {constB=-1})) {g=true}
  n33 = Sine(None=@n32) {period=0.8}
  n34 = Sine(None=@n30) {period=0.23}
  n35 = AppendVector(A=AppendVector(A=Constant(), B=Add(A=@n31, B=@n33)), B=Add(A=Multiply(A=@n34) {constB=0.2}, B=Sine(None=@n32) {period=0.6}))
  n36 = Multiply(
   A=Normalize(VectorInput=Transform(None=@n35) {transformSourceType=TRANSFORMSOURCE_Local})
   B=ScalarParameter() {defaultValue=0.2, parameterName=Wobble}
  )
  StaticSwitchParameter(True=@n25.WorldPositionOffset, False=Add(A=@n25.WorldPositionOffset, B=Multiply(A=@n27, B=@n36))) {parameterName=DEBUG Orientation}
 ]
}
